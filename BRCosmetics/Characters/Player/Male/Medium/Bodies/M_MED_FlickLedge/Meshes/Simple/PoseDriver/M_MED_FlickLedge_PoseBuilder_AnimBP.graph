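# Post-pose Animation Blueprint for the M_MED_FlickLedge character body cosmetic, built on that body's own skeleton.
# Copy Pose From Mesh pulls the pose and curves from the attached parent mesh and feeds the FlickLedge PoseBuilder Control Rig, which drives the output pose.
# The update event is empty, and the gravity, movement and rigid body settings variables are declared but not used by either graph.
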AnimBlueprint /BRCosmetics/Characters/Player/Male/Medium/Bodies/M_MED_FlickLedge/Meshes/Simple/PoseDriver/M_MED_FlickLedge_PoseBuilder_AnimBP
target_skeleton=/BRCosmetics/Characters/Player/Male/Medium/Bodies/M_MED_FlickLedge/Meshes/Simple/M_MED_FlickLedge_Skeleton
graphs [
 AnimGraph {nodes [
   n1 = Control Rig(Source=Copy Pose From Mesh() {bCopyCurves=true, bUseAttachedParent=true}) {controlRigAssetReference {blueprintRigClass=&/M_MED_FlickLedge_PoseBuilder_CtrlRig.M_MED_FlickLedge_PoseBuilder_CtrlRig_C}}
   Output Pose(Result=@n1) {layerGroup=None}
  ]}
 EventGraph {nodes [Event BlueprintUpdateAnimation()]}
]
variables [
 Gravity Override {type=Vector}
 Gravity Override_AnimDynamics {type=Vector}
 MoveFloorCollision {type=boolean}
 MovingForward {type=boolean}
 RBAN Settings Out {type=FortRigidBodyAnimNodeParameters}
 RBAN Settings Sliding {
  type=FortRigidBodyAnimNodeParameters
  default {simSpaceSettings {
    dampingAlpha=0
    maxAngularAcceleration=10
    maxAngularVelocity=4
    maxLinearAcceleration=1000
    maxLinearVelocity=200
    velocityScaleZ=0.4
    worldAlpha=0.4
   }}
 }
]
